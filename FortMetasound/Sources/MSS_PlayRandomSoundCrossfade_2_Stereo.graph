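# One-shot stereo MetaSound source that crossfades between random Close and Distant sounds by attenuation distance.
# Two MSS_PlayRandomSound_Stereo instances trigger together on play, sharing pitch and volume ranges.
# Distance, mapped through MinRange and MaxRange, reads two wavetable bank curves that set each layer's gain in a stereo mixer.
# A trigger accumulator signals finished once both layers end.
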MetaSoundSource /FortMetasound/Sources/MSS_PlayRandomSoundCrossfade_2_Stereo
inputs [
 Curves {type=WaveTableBankAsset}
 MaxPitch {type=Float}
 MaxRange {type=Float}
 MaxVolume {type=Float}
 MinPitch {type=Float}
 MinRange {type=Float}
 MinVolume {type=Float}
 Sounds.Close {type=WaveAsset:Array}
 Sounds.Distant {type=WaveAsset:Array}
 UE.Attenuation.Distance {type=Float}
 UE.Source.OnPlay {type=Trigger}
]
interfaces [UE.Attenuation:1.0, UE.OutputFormat.Stereo:1.0, UE.Source.OneShot:1.0, UE.Source:1.1]
outputs [
 "UE.OutputFormat.Stereo.Audio:0" {type=Audio}
 "UE.OutputFormat.Stereo.Audio:1" {type=Audio}
 UE.Source.OneShot.OnFinished {type=Trigger}
]
pages [Default {
  input_defaults {
   Curves=null
   MaxPitch=1.05
   MaxRange=1024
   MaxVolume=1.05
   MinPitch=0.95
   MinRange=0
   MinVolume=0.95
   UE.Attenuation.Distance=0
   UE.Source.OnPlay=null
   Sounds.Close []
   Sounds.Distant []
  }
  nodes [
   n1 = &MSS_PlayRandomSound_Stereo(MaxPitch=^MaxPitch, MaxVolume=^MaxVolume, MinPitch=^MinPitch, MinVolume=^MinVolume, Sounds=^Sounds.Close, UE.Source.OnPlay=^UE.Source.OnPlay) {also {defaults {Shared State Behavior=2}}}
   n2 = &MSS_PlayRandomSound_Stereo(
    MaxPitch=^MaxPitch
    MaxVolume=^MaxVolume
    MinPitch=^MinPitch
    MinVolume=^MinVolume
    Sounds=^Sounds.Distant
    UE.Source.OnPlay=^UE.Source.OnPlay
   ) {also {defaults {Shared State Behavior=2}}}
   AudioMixerAudioMixerStereo2_0 = "AudioMixer.Audio Mixer (Stereo, 2)"(
    In 0 L=@n1."UE.OutputFormat.Stereo.Audio:0"
    In 0 R=@n1."UE.OutputFormat.Stereo.Audio:1"
    In 1 L=@n2."UE.OutputFormat.Stereo.Audio:0"
    In 1 R=@n2."UE.OutputFormat.Stereo.Audio:1"
    Gain 0=UE.WaveTableBankEvaluate(WaveTableBank=^Curves, Input=MapRange.MapRange:Float(In Range B=^MaxRange, In Range A=^MinRange, In=^UE.Attenuation.Distance).Out Value).Output
    Gain 1=UE.WaveTableBankEvaluate(WaveTableBank=^Curves, Input=MapRange.MapRange:Float(In Range B=^MaxRange, In Range A=^MinRange, In=^UE.Attenuation.Distance).Out Value) {also {defaults {Index=1}}}.Output
   )
   TriggerAccumulatorTriggerAccumulate2_0 = "TriggerAccumulator.Trigger Accumulate (2)"(In 0=@n1.UE.Source.OneShot.OnFinished, In 1=@n2.UE.Source.OneShot.OnFinished) {also {defaults {Auto Reset="<MetaSoundBuilderResult.SUCCEEDED: 0>"}}}
  ]
  links [
   {dst="@UE.OutputFormat.Stereo.Audio:0", src=AudioMixerAudioMixerStereo2_0.Out L}
   {dst="@UE.OutputFormat.Stereo.Audio:1", src=AudioMixerAudioMixerStereo2_0.Out R}
   {dst="@UE.Source.OneShot.OnFinished", src=TriggerAccumulatorTriggerAccumulate2_0.out}
  ]
 }]
properties {assetImportData=&self:AssetImportData, numChannels=2, outputFormat=Stereo}
