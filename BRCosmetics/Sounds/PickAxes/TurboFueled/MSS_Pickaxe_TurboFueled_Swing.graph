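# One-shot stereo MetaSound for the TurboFueled pickaxe swing, with a configurable random start delay.
# A clamped SwingIdx triggers a tree of array concats to pick one of four swing sets; only the first holds sounds by default.
# A crossfade subpatch plays a random sound with pitch and volume variation; its distant array is left empty.
# An IsPreview branch chooses whether attenuation distance or a fixed value drives the distance input.
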MetaSoundSource /BRCosmetics/Sounds/PickAxes/TurboFueled/MSS_Pickaxe_TurboFueled_Swing
inputs [
 Cosmetics.IsPreview {type=Bool}
 PitchModulationRange.Max {type=Float}
 PitchModulationRange.Min {type=Float}
 StartDelay.Max {type=Time}
 StartDelay.Min {type=Time}
 Swing.Idx1.Sounds {type=WaveAsset:Array}
 Swing.Idx2.Sounds {type=WaveAsset:Array}
 Swing.Idx3.Sounds {type=WaveAsset:Array}
 Swing.Idx4.Sounds {type=WaveAsset:Array}
 SwingIdx {type=Int32}
 TemplatePitchMultiplier {type=Float}
 TemplateVolumeMultiplier {type=Float}
 UE.Attenuation.Distance {type=Float}
 UE.Source.OnPlay {type=Trigger}
 VolumeModulationRange.Max {type=Float}
 VolumeModulationRange.Min {type=Float}
]
interfaces [Fort.Cosmetics:1.0, UE.Attenuation:1.0, UE.OutputFormat.Stereo:1.0, UE.Source.OneShot:1.0, UE.Source:1.1]
outputs [
 "UE.OutputFormat.Stereo.Audio:0" {type=Audio}
 "UE.OutputFormat.Stereo.Audio:1" {type=Audio}
 UE.Source.OneShot.OnFinished {type=Trigger}
]
pages [Default {
  input_defaults {
   Cosmetics.IsPreview=null
   PitchModulationRange.Max=1.05
   PitchModulationRange.Min=0.95
   StartDelay.Max=0
   StartDelay.Min=0
   Swing.Idx2.Sounds=null
   Swing.Idx3.Sounds=null
   Swing.Idx4.Sounds=null
   SwingIdx=null
   TemplatePitchMultiplier=1
   TemplateVolumeMultiplier=1
   UE.Attenuation.Distance=0
   UE.Source.OnPlay=null
   VolumeModulationRange.Max=1.05
   VolumeModulationRange.Min=0.95
   Swing.Idx1.Sounds [
    &/Pickaxe_TurboFueled_Swing_01
    &/Pickaxe_TurboFueled_Swing_02
    &/Pickaxe_TurboFueled_Swing_03
    &/Pickaxe_TurboFueled_Swing_04
    &/Pickaxe_TurboFueled_Swing_05
    &/Pickaxe_TurboFueled_Swing_06
   ]
  }
  nodes [
   n1 = UE.RandomTime(Max=^StartDelay.Max, Min=^StartDelay.Min, Next=^UE.Source.OnPlay)
   TriggerDelay_0 = UE.Trigger Delay(In=@n1.On Next, Delay Time=@n1.Value)
   n2 = TriggerCompare.Clamp:Bool(A=^Cosmetics.IsPreview, Compare=@TriggerDelay_0.out) {also {defaults {B="<MetaSoundBuilderResult.SUCCEEDED: 0>"}}}
   TriggerRouteTriggerRouteFloat2_0 = "TriggerRoute.Trigger Route (Float, 2):Float"(Value 1=^UE.Attenuation.Distance, Set 1=@n2.False, Set 0=@n2.True)
   TriggerSelect4_0 = "UE.Trigger Select (4)"(In=^UE.Source.OnPlay, Index=Clamp.Clamp:Int32(In=^SwingIdx) {also {defaults {Max=3}}}.Value)
   TriggerAnyTriggerAny2_0 = "TriggerAny.Trigger Any (2)"(In 0=@TriggerSelect4_0.Out 2, In 1=@TriggerSelect4_0.Out 3)
   TriggerAnyTriggerAny2_1 = "TriggerAny.Trigger Any (2)"(In 0=@TriggerSelect4_0.Out 0, In 1=@TriggerSelect4_0.Out 1)
   TriggerAnyTriggerAny2_3 = "TriggerAny.Trigger Any (2)"(In 1=@TriggerAnyTriggerAny2_0.out, In 0=@TriggerAnyTriggerAny2_1.out)
   n3 = Array.Concat:WaveAsset:Array(
    Right Array=Array.Concat:WaveAsset:Array(
     Left Array=Array.Concat:WaveAsset:Array(Left Array=^Swing.Idx3.Sounds, Trigger=@TriggerSelect4_0.Out 2) {also {defaults {Right Array []}}}.Array
     Right Array=Array.Concat:WaveAsset:Array(Left Array=^Swing.Idx4.Sounds, Trigger=@TriggerSelect4_0.Out 3) {also {defaults {Right Array []}}}.Array
     Trigger=@TriggerAnyTriggerAny2_0.out
    ).Array
    Left Array=Array.Concat:WaveAsset:Array(
     Left Array=Array.Concat:WaveAsset:Array(Left Array=^Swing.Idx1.Sounds, Trigger=@TriggerSelect4_0.Out 0) {also {defaults {Right Array []}}}.Array
     Right Array=Array.Concat:WaveAsset:Array(Left Array=^Swing.Idx2.Sounds, Trigger=@TriggerSelect4_0.Out 1) {also {defaults {Right Array []}}}.Array
     Trigger=@TriggerAnyTriggerAny2_1.out
    ).Array
    Trigger=@TriggerAnyTriggerAny2_3.out
   )
   n4 = InitVariable.WaveAsset:Array()
   n5 = /FortMetasound/Sources/MSS_PlayRandomSoundCrossfade_2_Stereo(
    MaxVolume=^VolumeModulationRange.Max
    MinVolume=^VolumeModulationRange.Min
    MaxPitch=UE.Multiply:Float(PrimaryOperand=^PitchModulationRange.Max, AdditionalOperands=^TemplatePitchMultiplier).out
    MinPitch=UE.Multiply:Float(PrimaryOperand=^PitchModulationRange.Min, AdditionalOperands=^TemplatePitchMultiplier).out
    UE.Source.OnPlay=@TriggerRouteTriggerRouteFloat2_0.On Set
    UE.Attenuation.Distance=@TriggerRouteTriggerRouteFloat2_0.Value
    Sounds.Distant=VariableAccessor.WaveAsset:Array(Variable=InitVariable.WaveAsset:Array().Variable).Value
    Sounds.Close=VariableAccessor.WaveAsset:Array(Variable=VariableMutator.WaveAsset:Array(Value=@n3.Array, Variable=@n4.Variable).Variable).Value
   ) {also {defaults {Curves=/Game/Sounds/Attenuation/Curves/Pickaxe_Swing_DistanceCurves, MaxRange=3512, MinRange=512}}}
   AudioMixerAudioMixerStereo2_0 = "AudioMixer.Audio Mixer (Stereo, 2)"(Gain 0=^TemplateVolumeMultiplier, In 0 L=@n5."UE.OutputFormat.Stereo.Audio:0", In 0 R=@n5."UE.OutputFormat.Stereo.Audio:1")
   TriggerAnyTriggerAny2_2 = "TriggerAny.Trigger Any (2)"(In 0=@n5.UE.Source.OneShot.OnFinished)
  ]
  links [
   {dst="@UE.OutputFormat.Stereo.Audio:0", src=AudioMixerAudioMixerStereo2_0.Out L}
   {dst="@UE.OutputFormat.Stereo.Audio:1", src=AudioMixerAudioMixerStereo2_0.Out R}
   {dst="@UE.Source.OneShot.OnFinished", src=TriggerAnyTriggerAny2_2.out}
  ]
 }]
properties {
 assetImportData=&self:AssetImportData
 attenuationSettings=/Game/Sounds/Attenuation/ATT_Fort_Weapon_Pickaxe_NoAtt_Att
 audioPropertiesSheet=/BRCosmetics/Sounds/Settings/PropertySheets/Pickaxe/APS_Cosmetic_Pickaxe_Swing
 maxDistance=3612
 numChannels=2
 outputFormat=Stereo
 sampleRate=48000
 soundClassObject=&Fort_WeaponMelee_PSM
 volume=0.5
 concurrencySet [&Player_Melee_Weapon_Conc]
}
